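# Translucent unlit Niagara material for the LintMermaid emote's spiral effect, usable on sprites, ribbons and mesh particles.
# A static switch either subtracts an erosion texture scaled by the dynamic Erosion parameter from the main texture, or smoothsteps the main texture between Erosion and twice Erosion.
# The Feather parameter is declared but unused.
# Particle color tints the result through MF_FX_EmissiveAdjustments, and particle alpha scales it separately.
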Material /BRCosmetics/Animation/Game/MainPlayer/Emotes/LintMermaid/FX/M_LintMermaid_Spiral
graph {
 properties {
  bAutomaticallySetUsageInEditor=false
  bUsedWithNiagaraMeshParticles=true
  bUsedWithNiagaraRibbons=true
  bUsedWithNiagaraSprites=true
  blendMode=BLEND_Translucent
  shadingModel=MSM_Unlit
  usedShadingModels=Unlit
 }
 nodes [
  n1 = TextureSampleParameter2D(UVs=TextureCoordinate()) {parameterName=Main Texture, texture=&/T_LintMermaid_SpiralErosion}
  n2 = TextureSampleParameter2D(UVs=TextureCoordinate()) {parameterName=Erosion Texture, texture=&T_LintMermaid_SpiralErosion}
  n3 = DynamicParameter() {defaultValue=<rgba 0 1 1 1>, paramNames [Erosion,Feather,Param3,Param4]}
  n4 = DynamicParameter() {defaultValue=<rgba 0 1 1 1>, paramNames [Erosion,Feather,Param3,Param4]}
  n5 = StaticSwitchParameter(
   True=Subtract(A=@n1.RGB, B=Multiply(A=@n2.RGB, B=@n3.Erosion))
   False=MaterialFunctionCall(Alpha=@n1.RGB, Min=@n4.Erosion, Max=Add(A=@n4.Erosion, B=@n4.Erosion)) {materialFunction=&SmoothStep}.Result
  ) {parameterName=ErosionTexture?}
  MaterialFunctionCall(Emissive=Multiply(A=ParticleColor().RGB, B=@n5)) {materialFunction=&MF_FX_EmissiveAdjustments}
  Multiply(A=@n5, B=ParticleColor().A)
 ]
}
