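# Niagara fire effect for the LintMermaid emote: a short stylized flame that flares up, settles and fades.
# StylizedFire_Main uses two spawn groups with visibility tags across mesh and sprite renderers, shaped by scale, color and dynamic-material curves; its light-attributes module is disabled.
# StylizedFire_Ground adds an upward-facing flat sprite, and Glow a brief warm flash.
# GPU Embers burst and trickle out, driven by point velocity, drag and curl noise while cooling in color.
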NiagaraSystem /BRCosmetics/Animation/Game/MainPlayer/Emotes/LintMermaid/FX/NS_LintMermaid_Fire
dependencies {
 usedDataInterfaces [
  &NiagaraDataInterfaceCurve
  &NiagaraDataInterfaceColorCurve
  &NiagaraDataInterfaceMeshRendererInfo
  &NiagaraDataInterfaceVector2DCurve
  &NiagaraDataInterfaceVectorCurve
 ]
 usedDynamicInputs []
 usedModules [&SystemState]
}
emitters {
 Embers {
  simTarget=GPUComputeSim
  emitterUpdateScript [
   &EmitterState()
   &SpawnBurst_Instantaneous(Spawn Count=24, Use Spawn Probability=false, UseLoopCountLimit=false)
   &SpawnRate(
    SpawnRate=&FloatFromCurve(
     FloatCurve=DataInterface() {propertyValues {Curve {keys [<at 0 1>,<at 0.5 1>,<at 0.7 0>]}}}
     CurveIndex=$Emitter.NormalizedLoopAge
     Scale Curve=12
    )
    Use Spawn Probability=false
   )
  ]
  particleSpawnScript [
   &InitializeParticle(
    Color Maximum=<rgba 1 0.8902 0 1>
    Color Minimum=<rgba 1 0.0818 0 1>
    Color Mode=::NewEnumerator2
    Lifetime Max=1.2
    Lifetime Min=0.3
    Lifetime Mode=::NewEnumerator1
    Position Offset=<0 0 2>
    Sprite Size Mode=::NewEnumerator1
    Uniform Sprite Size Max=3
    Uniform Sprite Size Min=1
    UsePositionOffset=true
    Write Material Random=true
   ) {module=InitializeParticle001}
   &ShapeLocation(Disc Coverage=1, MaskSpawnGroup=false, Override Seed=false, Ring Radius=10, Shape Primitive=::NewEnumerator4, UseOwnerScale=true)
   &AddVelocityFromPoint(
    Module.Use Falloff Distance=false
    Origin Offset=<-15 0 -45>
    Velocity Strength=&RandomRangeFloat(Minimum=75, Maximum=325)
   )
   &CameraOffset(Camera Offset Amount=-10)
  ]
  particleUpdateScript [
   &ParticleState()
   &Drag(Drag=&RandomRangeFloat(Minimum=3, Maximum=5), Use Linear Drag=true, Use Rotational Drag=false)
   /Niagara/Modules/Update/Forces/CurlNoiseForce(
    Noise Strength=&FloatFromCurve(
     FloatCurve=DataInterface() {propertyValues {Curve {keys [<at 0 1>,<at 0.1 0.25>,<at 0.1198 0.98>,<at 0.2002 0.21>,<at 0.2421 0.9996>,<at 1 0>]}}}
     Scale Curve=&RandomRangeFloat(Minimum=300, Maximum=500)
    )
    Pan Noise=true
    Pan Noise Field=<0 0 1>
   )
   &SolveForcesAndVelocity(Clamp Velocity=false, Limit Acceleration=false)
   &ScaleColor(
    Scale Alpha=&FloatFromCurve(FloatCurve=DataInterface() {propertyValues {Curve {keys [<at 0 0 RCIM_Cubic>,<at 0.25 1 RCIM_Cubic>,<at 1 0 RCIM_Cubic>]}}})
    Scale RGB=&MakeVectorFromLinearColorRGB(Color=&ColorFromCurve(ColorCurve=DataInterface() {propertyValues {
        AlphaCurve {keys [<at 0 1 RCIM_Cubic>,<at 0.7107 1>]}
        BlueCurve {keys [<at 0 0 RCIM_Cubic>,<at 0.7114 0 RCIM_Cubic>]}
        GreenCurve {keys [<at 0 5.2565 RCIM_Cubic>,<at 0.7114 0.6624 RCIM_Cubic>]}
        RedCurve {keys [<at 0 15 RCIM_Cubic>,<at 0.7114 1 RCIM_Cubic>]}
       }}))
    ScaleA=true
    ScaleRGB=true
   )
   &ScaleSpriteSizeBySpeed(Max Scale Factor=<0.2 3>, Velocity Threshold=180) {module=ScaleSpriteSizeBySpeed001}
  ]
  rendererClasses [&NiagaraSpriteRendererProperties]
  renderers [{rendererClass=&NiagaraSpriteRendererProperties}]
 }
 Glow {
  emitterUpdateScript [
   &EmitterState()
   &SpawnBurst_Instantaneous(Spawn Time=0.04, Use Spawn Probability=false, UseLoopCountLimit=false)
  ]
  particleSpawnScript [
   &InitializeParticle(
    Color=<rgba 5 1.3673 0 0.4>
    Lifetime=0.54
    Position Offset=<0 -2 8>
    Sprite Size=<60 80>
    Sprite Size Mode=::NewEnumerator4
    UsePositionOffset=true
    Write Material Random=true
   )
   &CameraOffset(Camera Offset Amount=-15)
  ]
  particleUpdateScript [
   &ParticleState()
   &ScaleColor(
    Linear Color Curve=DataInterface() {propertyValues {
      AlphaCurve {keys [
        <at 0 0>
        {arriveTangent=0.1697, interpMode=RCIM_Cubic, leaveTangent=0.1697, leaveTangentWeight=0.3333, tangentMode=RCTM_User, time=0.25, value=1}
        {arriveTangent=-1.8982, interpMode=RCIM_Cubic, leaveTangent=-1.8982, tangentMode=RCTM_User, time=1, value=0}
       ]}
      BlueCurve {keys [<at 0 1>,<at 1 1>]}
      GreenCurve {keys [<at 0 1>,<at 1 1>]}
      RedCurve {keys [<at 0 1>,<at 1 1>]}
     }}
    Scale Mode=::NewEnumerator2
   )
  ]
  rendererClasses [&NiagaraSpriteRendererProperties]
  renderers [{rendererClass=&NiagaraSpriteRendererProperties}]
 }
 StylizedFire_Ground {
  emitterUpdateScript [
   &EmitterState()
   &SpawnBurst_Instantaneous(Use Spawn Probability=false, UseLoopCountLimit=false)
  ]
  particleSpawnScript [
   &InitializeParticle(
    Lifetime=1.6
    Position Offset=<0 0 1>
    Sprite Size Mode=::NewEnumerator3
    Uniform Sprite Size=22.5
    UsePositionOffset=true
    Write Material Random=true
   )
   &SpriteFacingAndAlignment(Sprite Facing=<0 0 1>, WriteAlignment=false, WriteFacing=true)
  ]
  particleUpdateScript [
   &ParticleState()
   &DynamicMaterialParameters(
    Index 0 Param 2=2
    Index 0 Param 3=0.15
    Module.Param0WriteEnabled=true
    Module.Param0WriteEnabled1=true
    Module.Param1WriteEnabled=true
    Module.Param1WriteEnabled1=true
    Module.Param2WriteEnabled=true
    Module.Param2WriteEnabled1=true
    Module.Param3WriteEnabled=true
    Module.Param3WriteEnabled1=true
    Write Parameter Index 1=true
   )
   &ScaleColor(
    Linear Color Curve=DataInterface() {propertyValues {
      AlphaCurve {keys [<at 0 1>,<at 0.8 1>,<at 1 0>]}
      BlueCurve {keys [<at 0 1>,<at 0.1237 5>,<at 0.4 1>,<at 0.8 1>]}
      GreenCurve {keys [<at 0 1>,<at 0.1237 5>,<at 0.4 1>,<at 0.8 1>]}
      RedCurve {keys [<at 0 1>,<at 0.1237 5>,<at 0.4 1>,<at 0.8 1>]}
     }}
    Scale Mode=::NewEnumerator2
   )
  ]
  rendererClasses [&NiagaraSpriteRendererProperties]
  renderers [{rendererClass=&NiagaraSpriteRendererProperties}]
 }
 StylizedFire_Main {
  emitterUpdateScript [
   &EmitterState()
   &SpawnBurst_Instantaneous(Use Spawn Probability=false, UseLoopCountLimit=false)
   &SpawnBurst_Instantaneous(Spawn Group=1, Use Spawn Probability=false, UseLoopCountLimit=false) {module=SpawnBurst_Instantaneous001}
  ]
  particleSpawnScript [
   /SetVariables(Particles.VisibilityTag=&MaskIntBySpawnGroup(Masked Int=0, Passthrough Int=1, Spawn Group Mask=0)) {isSetParameters=true}
   &InitializeParticle(
    Lifetime=1.6
    Mesh Scale=<0.35 0.065 1>
    Mesh Scale Mode=::NewEnumerator4
    Position Offset=&MaskVectorBySpawnGroup(Masked Vector=<0 0 0>, Passthrough Vector=<0 0 -2.5>)
    Sprite Size=<20 35>
    Sprite Size Mode=::NewEnumerator4
    UsePositionOffset=true
    Write Material Random=true
   )
   &InitialMeshOrientation(Mesh Axis To Orient=<0 0 1>, Mesh Orientation Mode=::NewEnumerator2, Rotation=<0 0 -0.25>, Use Rotation Vector=true)
   &SpriteFacingAndAlignment(Sprite Alignment=<0 0 1>, WriteAlignment=true, WriteFacing=false)
  ]
  particleUpdateScript [
   &ParticleState()
   &ScaleSpriteSize(
    Non-Uniform Curve Sprite Scale=DataInterface() {propertyValues {
      XCurve {keys [
        {arriveTangent=8.1644, interpMode=RCIM_Cubic, leaveTangent=8.1644, leaveTangentWeight=0.3333, tangentMode=RCTM_User, time=0, value=0}
        {
         arriveTangent=-0.3076
         arriveTangentWeight=0.0673
         interpMode=RCIM_Cubic
         leaveTangent=-0.3076
         leaveTangentWeight=0.2917
         tangentMode=RCTM_User
         time=0.075
         value=1.25
        }
        {arriveTangent=-0.3448, interpMode=RCIM_Cubic, leaveTangent=-0.3448, time=0.6, value=1}
        {
         arriveTangent=-0.0864
         arriveTangentWeight=0.1333
         interpMode=RCIM_Cubic
         leaveTangent=-0.0864
         leaveTangentWeight=0.0667
         tangentMode=RCTM_User
         time=0.8
         value=1
        }
       ]}
      YCurve {keys [
        {arriveTangent=8.1644, interpMode=RCIM_Cubic, leaveTangent=8.1644, leaveTangentWeight=0.3333, tangentMode=RCTM_User, time=0, value=0}
        {arriveTangent=-0.3076, interpMode=RCIM_Cubic, leaveTangent=-0.3076, tangentMode=RCTM_User, time=0.075, value=2.5}
        {arriveTangent=-2.8571, interpMode=RCIM_Cubic, leaveTangent=-2.8571, time=0.25, value=1.5}
        {arriveTangent=-0.2096, interpMode=RCIM_Cubic, leaveTangent=-0.2096, tangentMode=RCTM_User, time=0.6, value=1}
        {
         arriveTangent=-0.0864
         arriveTangentWeight=0.1333
         interpMode=RCIM_Cubic
         leaveTangent=-0.0864
         leaveTangentWeight=0.0667
         tangentMode=RCTM_User
         time=0.8
         value=1
        }
       ]}
     }}
    Scale Sprite Size Mode=::NewEnumerator3
   )
   &ScaleMeshSize(Scale Factor=&VectorFromCurve(VectorCurve=DataInterface() {propertyValues {
       XCurve {keys [
         {arriveTangent=8.1644, interpMode=RCIM_Cubic, leaveTangent=8.1644, leaveTangentWeight=0.3333, tangentMode=RCTM_User, time=0, value=0}
         {
          arriveTangent=-0.0176
          arriveTangentWeight=0.025
          interpMode=RCIM_Cubic
          leaveTangent=-0.0176
          leaveTangentWeight=0.1083
          tangentMode=RCTM_User
          time=0.075
          value=1.25
         }
         {arriveTangent=-0.3448, interpMode=RCIM_Cubic, leaveTangent=-0.3448, time=0.6, value=1}
         {interpMode=RCIM_Cubic, leaveTangent=-0.8333, time=0.8, value=1}
        ]}
       YCurve {keys [
         {arriveTangent=8.1644, interpMode=RCIM_Cubic, leaveTangent=8.1644, leaveTangentWeight=0.2667, tangentMode=RCTM_User, time=0, value=0}
         {
          arriveTangent=0.0046
          arriveTangentWeight=0.025
          interpMode=RCIM_Cubic
          leaveTangent=0.0046
          leaveTangentWeight=0.1083
          tangentMode=RCTM_User
          time=0.075
          value=2.5
         }
         {arriveTangent=-2.8571, interpMode=RCIM_Cubic, leaveTangent=-2.8571, time=0.25, value=1.5}
         {
          arriveTangent=-0.0595
          arriveTangentWeight=0.1083
          interpMode=RCIM_Cubic
          leaveTangent=-0.0595
          leaveTangentWeight=0.1167
          tangentMode=RCTM_User
          time=0.6
          value=1
         }
         {arriveTangent=-0.0864, arriveTangentWeight=0.1167, interpMode=RCIM_Cubic, leaveTangent=-0.0864, tangentMode=RCTM_User, time=0.8, value=1}
        ]}
       ZCurve {keys [
         {arriveTangent=8.1644, interpMode=RCIM_Cubic, leaveTangent=8.1644, leaveTangentWeight=0.3333, tangentMode=RCTM_User, time=0, value=0}
         {arriveTangent=0.0757, interpMode=RCIM_Cubic, leaveTangent=0.0757, tangentMode=RCTM_User, time=0.125, value=1.6}
         {arriveTangent=-0.0595, interpMode=RCIM_Cubic, leaveTangent=-0.0595, tangentMode=RCTM_User, time=0.45, value=1}
         {arriveTangent=-0.0864, interpMode=RCIM_Cubic, leaveTangent=-0.0864, tangentMode=RCTM_User, time=0.8, value=1}
        ]}
      }}))
   &DynamicMaterialParameters(
    Index 2 Param 1=&FloatFromCurve(FloatCurve=DataInterface() {propertyValues {Curve {keys [
         {arriveTangent=5.1543, interpMode=RCIM_Cubic, leaveTangent=5.1543, tangentMode=RCTM_User, time=0, value=1}
         {arriveTangent=0.023, arriveTangentWeight=0.0667, interpMode=RCIM_Cubic, leaveTangent=0.023, tangentMode=RCTM_User, time=0.2, value=1}
         {arriveTangent=0.0022, arriveTangentWeight=0.1667, interpMode=RCIM_Cubic, leaveTangent=0.0022, tangentMode=RCTM_User, time=0.7, value=1}
         {arriveTangentWeight=0.1, interpMode=RCIM_Cubic, tangentMode=RCTM_User, time=1, value=0}
        ]}}})
    Module.Param0WriteEnabled=true
    Module.Param0WriteEnabled1=true
    Module.Param0WriteEnabled2=true
    Module.Param1WriteEnabled=true
    Module.Param1WriteEnabled1=true
    Module.Param1WriteEnabled2=true
    Module.Param2WriteEnabled=true
    Module.Param2WriteEnabled1=true
    Module.Param2WriteEnabled2=true
    Module.Param3WriteEnabled=true
    Module.Param3WriteEnabled1=true
    Module.Param3WriteEnabled2=true
    Write Parameter Index 1=true
    Write Parameter Index 2=true
   )
   &ScaleColor(
    Linear Color Curve=DataInterface() {propertyValues {
      AlphaCurve {keys [
        {arriveTangent=5.1543, interpMode=RCIM_Cubic, leaveTangent=5.1543, tangentMode=RCTM_User, time=0, value=0}
        {arriveTangent=0.023, interpMode=RCIM_Cubic, leaveTangent=0.023, tangentMode=RCTM_User, time=0.125, value=1}
        {arriveTangent=0.0022, interpMode=RCIM_Cubic, leaveTangent=0.0022, tangentMode=RCTM_User, time=0.7, value=1}
        {interpMode=RCIM_Cubic, tangentMode=RCTM_User, time=1, value=0}
       ]}
      BlueCurve {keys [<at 0 1>,<at 0.1 3>,<at 0.45 3>,<at 0.6 1>]}
      GreenCurve {keys [<at 0 1>,<at 0.1 3>,<at 0.45 3>,<at 0.6 1>]}
      RedCurve {keys [<at 0 1>,<at 0.1 3>,<at 0.45 3>,<at 0.6 1>]}
     }}
    Scale Mode=::NewEnumerator2
   )
   &FN_Light_Attributes(
    Color=<rgba 1 0.437 0 1>
    Create Color Attribute=true
    Create Exponent Attribute=true
    Create Position Attribute=true
    Create Radius Attribute=true
    Exponent=4
    Position=&AddVectorToPosition(Position=$Engine.Emitter.SimulationPosition, Vector=<0 0 20>)
    Radius=120
    Scale Color=&FloatFromCurve(
     FloatCurve=DataInterface() {propertyValues {Curve {keys [
         <at 0 0 RCIM_Cubic>
         {arriveTangent=5, interpMode=RCIM_Cubic, leaveTangent=5, time=0.1, value=2.26}
         {arriveTangent=-0.0317, interpMode=RCIM_Cubic, leaveTangent=-0.0317, tangentMode=RCTM_User, time=0.2, value=1}
         {arriveTangent=-1.25, interpMode=RCIM_Cubic, leaveTangent=-1.25, time=0.8, value=1}
         <at 1 0 RCIM_Cubic>
        ]}}}
     Scale Curve=5
    )
   ) {enabled=false}
  ]
  rendererClasses [
   &NiagaraMeshRendererProperties
   &NiagaraLightRendererProperties
   &NiagaraSpriteRendererProperties
  ]
  renderers [
   {rendererClass=&NiagaraMeshRendererProperties, rendererIndex=0}
   {rendererClass=&NiagaraLightRendererProperties, rendererIndex=1}
   {rendererClass=&NiagaraSpriteRendererProperties, rendererIndex=2}
  ]
 }
}
summary {
 systemName=NS_LintMermaid_Fire
 emitters [
  {emitterName=Embers, simTarget=GPUComputeSim, rendererClasses [&NiagaraSpriteRendererProperties]}
  {emitterName=Glow, rendererClasses [&NiagaraSpriteRendererProperties]}
  {
   emitterName=StylizedFire_Main
   rendererClasses [
    &NiagaraMeshRendererProperties
    &NiagaraLightRendererProperties
    &NiagaraSpriteRendererProperties
   ]
  }
  {emitterName=StylizedFire_Ground, rendererClasses [&NiagaraSpriteRendererProperties]}
 ]
 userVariables []
}
