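# Niagara system for the LintMermaid emote that shows mac and cheese pouring and splatting, with every emitter positioned by skeletal mesh sampling.
# Two ribbon Goop emitters and toon-outlined, noise-driven falling macaroni meshes spawn briefly early in each loop.
# Delayed bursts add the impact: a scaling cheese splat mesh that follows the skeleton, sprite droplets, and macaroni flung upward under gravity.
# Goop and splat colors adapt to time of day.
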NiagaraSystem /BRCosmetics/Animation/Game/MainPlayer/Emotes/LintMermaid/FX/NS_LintMermaid_MacCheese
dependencies {
 usedDataInterfaces [&NiagaraDataInterfaceVectorCurve, &NiagaraDataInterfaceCurve, &NiagaraDataInterfaceSkeletalMesh]
 usedDynamicInputs []
 usedModules [
  &SystemState
  &self:SystemSpawnScript.SystemScriptSource.SystemScriptGraph.NiagaraNodeAssignment_0.SetVariables_#1
 ]
}
emitters {
 Burst_Splashes {
  emitterUpdateScript [
   &EmitterState(
    Enable Distance Culling=true
    Life Cycle Mode=::NewEnumerator1
    Loop Behavior=::NewEnumerator1
    Loop Delay=0.24
    Loop Duration=1
    MaxDistance=12000
    Module.bUseMaxDistance=true
    Module.bUseMinDistance=false
    Scalability Mode=::NewEnumerator1
    UseLoopDelay=true
   )
   &SpawnBurst_Instantaneous(Spawn Count=5, Use Spawn Probability=false)
  ]
  particleSpawnScript [
   &InitializeParticle(
    Color=<rgba 0.6094 0.239 0.0526 1>
    Lifetime Max=0.8
    Lifetime Min=0.3
    Lifetime Mode=::NewEnumerator1
    Position Offset=<0 0 -100>
    Sprite Rotation Mode=::NewEnumerator2
    Sprite Size Mode=::NewEnumerator1
    Uniform Sprite Size Max=4
    Uniform Sprite Size Min=1
    UsePositionOffset=false
    Write Material Random=true
   ) {module=InitializeParticle001}
   &SkeletalMeshLocation(Bone Sampling Mode=::NewEnumerator4, MaskSpawnGroup=false, Sampled Position Offset=<0 0 4>, Skeletal Mesh=$System.Skeletal Mesh)
   &ShapeLocation(
    MaskSpawnGroup=false
    Offset=<0 -8 -3>
    Offset Mode=::NewEnumerator0
    Override Seed=false
    Sphere Radius=5
    Sphere Surface Distribution=1
    UseOwnerScale=true
   )
   &AddVelocity(
    Module.Use Falloff Distance=false
    Origin Offset=&RandomRangeVector(Minimum=<0 0 -100>, Maximum=<0 0 0>)
    Override Seed=false
    Velocity Mode=::NewEnumerator1
    Velocity Speed=&RandomRangeFloat(Minimum=25, Maximum=40)
   )
   &SubUVAnimation(End Frame=3, SubUV Animation Mode:ENiagara_SubUVLookupMode=::NewEnumerator2)
  ]
  particleUpdateScript [
   &ParticleState()
   &Drag(Drag=&RandomRangeFloat(Minimum=0.8, Maximum=1.2), Use Linear Drag=true, Use Rotational Drag=false)
   &GravityForce(Coordinate Space=::Local, Gravity=<0 0 -80>)
   &SolveForcesAndVelocity(Clamp Velocity=false, Limit Acceleration=false)
   &ScaleSpriteSize(Scale Factor=&Vector2DFromFloat(Value=&FloatFromCurve(
      FloatCurve=DataInterface() {propertyValues {Curve {keys [
          {arriveTangent=12.5197, interpMode=RCIM_Cubic, leaveTangent=12.5197, tangentMode=RCTM_User, time=0, value=0.4376}
          {arriveTangent=0.8929, interpMode=RCIM_Cubic, leaveTangent=0.8929, tangentMode=RCTM_User, time=1, value=5}
         ]}}}
      Scale Curve=2
     )))
   &ScaleColor(
    Scale Alpha=&FloatFromCurve(FloatCurve=DataInterface() {propertyValues {Curve {keys [<at 0 -1>,<at 0.009 0.9316>,<at 0.7102 -0.1519>,<at 1 -1>]}}})
    Scale RGB=&VectorFromCurve(VectorCurve=DataInterface() {propertyValues {XCurve {keys [<at 0 2>,<at 1 0>]}, YCurve {keys [<at 0 2>,<at 1 0>]}, ZCurve {keys [<at 0 2>,<at 1 0>]}}})
    ScaleA=true
    ScaleRGB=true
   )
   &SpriteRotationRate(Rotation Rate=&Lerp_Float(A=-150, B=150, Alpha=$Particles.Initial.MaterialRandom))
   &DynamicMaterialParameters(
    Index 0 Param 1=&FloatFromCurve(FloatCurve=DataInterface() {propertyValues {Curve {keys [
         {arriveTangent=-2.1712, interpMode=RCIM_Cubic, leaveTangent=-2.1712, tangentMode=RCTM_User, time=0.25, value=1}
         {arriveTangent=-0.3234, interpMode=RCIM_Cubic, leaveTangent=-0.3234, tangentMode=RCTM_User, time=1.0011, value=0}
        ]}}})
    Index 0 Param 2=&FloatFromCurve(FloatCurve=DataInterface() {propertyValues {Curve {keys [<at -0.0086 -1.3001>,<at 0.9945 1.9153>]}}})
    Module.Param0WriteEnabled=true
    Module.Param1WriteEnabled=false
    Module.Param2WriteEnabled=true
    Module.Param3WriteEnabled=true
   )
  ]
  rendererClasses [&NiagaraSpriteRendererProperties]
  renderers [{rendererClass=&NiagaraSpriteRendererProperties}]
 }
 CheeseSplat {
  emitterUpdateScript [
   &EmitterState(
    Inactive Response=::NewEnumerator1
    Life Cycle Mode=::NewEnumerator1
    Loop Behavior=::NewEnumerator1
    Loop Delay=0.24
    Loop Duration=1.2
    UseLoopDelay=true
   )
   &SpawnBurst_Instantaneous(Use Spawn Probability=false, UseLoopCountLimit=false)
  ]
  particleSpawnScript [
   &InitializeParticle(
    Color=&FNTimeOfDayColor(
     Night Color=<rgba 0.4583 0.1402 0.0038 1>
     Morning Color=<rgba 1 0.2379 0 1>
     Day Color=<rgba 1 0.2379 0 1>
     Evening Color=<rgba 0.4583 0.1402 0.0038 1>
    )
    Lifetime=0.48
    Mesh Scale Mode=::NewEnumerator3
    Mesh Uniform Scale=0.1
    Sprite Size=<40 65>
    Sprite Size Mode=::NewEnumerator4
    UsePositionOffset=false
    Write Material Random=true
   )
   &InitialMeshOrientation(
    Mesh Orientation Mode=::NewEnumerator2
    Orientation Vector=<0 0 1>
    Rotation=&RandomRangeVector(Minimum=<0 0 0>, Maximum=<1 0 0>)
    Use Rotation Vector=true
   )
  ]
  particleUpdateScript [
   &ParticleState()
   &SkeletalMeshLocation(Bone Sampling Mode=::NewEnumerator4, MaskSpawnGroup=false, Skeletal Mesh=$System.Skeletal Mesh)
   &ScaleMeshSize(Scale Factor=&VectorFromCurve(
     VectorCurve=DataInterface() {propertyValues {
       XCurve {keys [<at 0 0.5 RCIM_Cubic>, {arriveTangent=-0.5, interpMode=RCIM_Cubic, leaveTangent=-0.5, time=0.5, value=1}, <at 1 0 RCIM_Cubic>]}
       YCurve {keys [<at 0 0.5 RCIM_Cubic>, {interpMode=RCIM_Cubic, leaveTangent=-0.5, time=0.5, value=1}]}
       ZCurve {keys [<at 0 0.5 RCIM_Cubic>, {interpMode=RCIM_Cubic, leaveTangent=-0.5, time=0.5, value=1}]}
      }}
     Scale Curve=<1 1 1.2>
    ))
   &SolveForcesAndVelocity(Clamp Velocity=false, Limit Acceleration=false)
   &ScaleColor(
    Scale Alpha=&FloatFromCurve(FloatCurve=DataInterface() {propertyValues {Curve {keys [
         {interpMode=RCIM_Cubic, tangentMode=RCTM_User, time=0, value=0.8}
         {interpMode=RCIM_Cubic, tangentMode=RCTM_User, time=0.5, value=1}
         {tangentMode=RCTM_User, time=1, value=0}
        ]}}})
    ScaleA=true
    ScaleRGB=true
   )
   &DynamicMaterialParameters(
    Index 1 Param 1=&FloatFromCurve(FloatCurve=DataInterface() {propertyValues {Curve {keys [
         {arriveTangent=-0.0643, interpMode=RCIM_Cubic, leaveTangent=-0.0643, tangentMode=RCTM_User, time=0, value=0}
         {arriveTangent=1.6749, interpMode=RCIM_Cubic, leaveTangent=1.6749, tangentMode=RCTM_User, time=1, value=1}
        ]}}})
    Index 2 Param 1=1
    Index 2 Param 2=5
    Module.Index1Param0Write=true
    Module.Index1Param1Write=true
    Module.Index1Param2Write=true
    Module.Index1Param3Write=true
    Module.Index2Param0Write=true
    Module.Index2Param1Write=true
    Module.Index2Param2Write=true
    Module.Index2Param3Write=true
    Write Parameter Index 0=false
    Write Parameter Index 1=true
    Write Parameter Index 2=true
   )
  ]
  rendererClasses [&NiagaraMeshRendererProperties]
  renderers [{rendererClass=&NiagaraMeshRendererProperties, rendererIndex=0}]
 }
 Goop {
  emitterSpawnScript [/SetVariables(Emitter.Distortion=0.65) {isSetParameters=true}]
  emitterUpdateScript [
   &EmitterState()
   &SpawnRate(
    SpawnRate=&FloatFromCurve(
     FloatCurve=DataInterface() {propertyValues {Curve {keys [<at 0 1>,<at 0.1 1>,<at 0.2 1>,<at 0.25 0>]}}}
     CurveIndex=$Emitter.NormalizedLoopAge
     Scale Curve=150
    )
    Use Spawn Probability=false
   )
  ]
  particleSpawnScript [
   &InitializeParticle(
    Color=&FNTimeOfDayColor(
     Night Color=<rgba 0.4896 0.3876 0.2856 1>
     Morning Color=<rgba 1 0.6119 0.4167 1>
     Day Color=<rgba 1 0.6119 0.4167 1>
     Evening Color=<rgba 0.4896 0.3876 0.2856 1>
    )
    Lifetime=0.24
    Ribbon Width=8
    Ribbon Width Mode=::NewEnumerator1
    UsePositionOffset=false
    Write Material Random=true
   )
   &SkeletalMeshLocation(Bone Sampling Mode=::NewEnumerator4, MaskSpawnGroup=false, Skeletal Mesh=$System.Skeletal Mesh)
   &CameraOffset(Camera Offset Amount=-15) {enabled=false}
  ]
  particleUpdateScript [
   &ParticleState()
   &SolveForcesAndVelocity(Clamp Velocity=false, Limit Acceleration=false)
   &DynamicMaterialParameters(
    Index 0 Param 2=&FloatFromCurve(FloatCurve=DataInterface() {propertyValues {Curve {keys [
         <at 0 0.3>
         {arriveTangent=-2.5521, interpMode=RCIM_Cubic, leaveTangent=-2.5521, tangentMode=RCTM_User, time=0.1, value=0.75}
         {arriveTangent=-0.0873, interpMode=RCIM_Cubic, leaveTangent=0.0167, tangentMode=RCTM_Break, time=1, value=0}
        ]}}})
    Index 0 Param 3=1
    Index 0 Param 4=1
    Module.Param0WriteEnabled=true
    Module.Param1WriteEnabled=true
    Module.Param2WriteEnabled=true
    Module.Param3WriteEnabled=true
   )
   &ScaleRibbonWidth(Ribbon Width Scale=&FloatFromCurve(FloatCurve=DataInterface() {propertyValues {Curve {keys [
         {arriveTangent=0.3835, interpMode=RCIM_Cubic, leaveTangent=0.3835, tangentMode=RCTM_User, time=0, value=0.85}
         {arriveTangent=4.7926, interpMode=RCIM_Cubic, leaveTangent=1.8929, tangentMode=RCTM_Break, time=0.3, value=1.75}
         <at 0.5 2>
        ]}}})) {module=ScaleRibbonWidth001}
   &ScaleColor(
    Scale Alpha=&Multiply_Float(
     A=&FNTimeOfDayScalar(Night Scale=1.5, Morning Scale=2, Evening Scale=2)
     B=&FloatFromCurve(FloatCurve=DataInterface() {propertyValues {Curve {keys [<at 0 0.8>,<at 0.3 0.95>,<at 1 1>]}}})
    )
    ScaleA=true
    ScaleRGB=true
   )
  ]
  rendererClasses [&NiagaraRibbonRendererProperties]
  renderers [{rendererClass=&NiagaraRibbonRendererProperties}]
 }
 Goop001 {
  emitterSpawnScript [/SetVariables() {isSetParameters=true}]
  emitterUpdateScript [
   &EmitterState()
   &SpawnRate(
    SpawnRate=&FloatFromCurve(
     FloatCurve=DataInterface() {propertyValues {Curve {keys [<at 0 1>,<at 0.1 1>,<at 0.2 1>,<at 0.25 0>]}}}
     CurveIndex=$Emitter.NormalizedLoopAge
     Scale Curve=150
    )
    Use Spawn Probability=false
   )
  ]
  particleSpawnScript [
   &InitializeParticle(
    Color=&FNTimeOfDayColor(
     Night Color=<rgba 0.4896 0.3876 0.2856 1>
     Morning Color=<rgba 1 0.6119 0.4167 1>
     Day Color=<rgba 1 0.6119 0.4167 1>
     Evening Color=<rgba 0.4896 0.3876 0.2856 1>
    )
    Lifetime=0.08
    Ribbon Width=18
    Ribbon Width Mode=::NewEnumerator1
    UsePositionOffset=false
    Write Material Random=true
   )
   &SkeletalMeshLocation(Bone Sampling Mode=::NewEnumerator4, MaskSpawnGroup=false, Skeletal Mesh=$System.Skeletal Mesh)
   &CameraOffset(Camera Offset Amount=-15) {enabled=false}
  ]
  particleUpdateScript [
   &ParticleState()
   &SolveForcesAndVelocity(Clamp Velocity=false, Limit Acceleration=false)
   &DynamicMaterialParameters(
    Index 0 Param 2=&FloatFromCurve(FloatCurve=DataInterface() {propertyValues {Curve {keys [
         <at 0 0.3>
         {arriveTangent=-2.5521, interpMode=RCIM_Cubic, leaveTangent=-2.5521, tangentMode=RCTM_User, time=0.1, value=0.75}
         {arriveTangent=-0.0873, interpMode=RCIM_Cubic, leaveTangent=0.0167, tangentMode=RCTM_Break, time=1, value=0}
        ]}}})
    Index 0 Param 3=1
    Index 0 Param 4=1
    Module.Param0WriteEnabled=true
    Module.Param1WriteEnabled=true
    Module.Param2WriteEnabled=true
    Module.Param3WriteEnabled=true
   )
   &ScaleRibbonWidth(Ribbon Width Scale=&FloatFromCurve(FloatCurve=DataInterface() {propertyValues {Curve {keys [
         {arriveTangent=0.3835, interpMode=RCIM_Cubic, leaveTangent=0.3835, tangentMode=RCTM_User, time=0, value=0.85}
         {arriveTangent=4.7926, interpMode=RCIM_Cubic, leaveTangent=1.8929, tangentMode=RCTM_Break, time=0.3, value=1.75}
         <at 0.5 2>
         <at 0.7 1>
        ]}}})) {module=ScaleRibbonWidth001}
   &ScaleColor(
    Scale Alpha=&Multiply_Float(
     A=&FNTimeOfDayScalar(Night Scale=1.5, Morning Scale=2, Evening Scale=2)
     B=&FloatFromCurve(FloatCurve=DataInterface() {propertyValues {Curve {keys [<at 0 0.8>,<at 0.3 0.95>,<at 1 1>]}}})
    )
    ScaleA=true
    ScaleRGB=true
   )
  ]
  rendererClasses [&NiagaraRibbonRendererProperties]
  renderers [{rendererClass=&NiagaraRibbonRendererProperties}]
 }
 MacFalling {
  emitterSpawnScript [
   /Game/Effects/Niagara/MasterMaterials/ToonShader/Template/NM_ToonShaderParameters(RimLight_Color:LinearColor=<rgba 2 1.3024 0.2399 1>)
   /Game/Effects/Niagara/MasterMaterials/ToonShader/Template/NM_OutlineParameters(OutlineColor:LinearColor=<rgba 0.0469 0.0289 0 1>, OutlineWidth=0.5)
  ]
  emitterUpdateScript [
   &EmitterState()
   &SpawnRate(
    SpawnRate=&FloatFromCurve(
     FloatCurve=DataInterface() {propertyValues {Curve {keys [<at 0.09 0>,<at 0.1 1>,<at 0.2 1>,<at 0.25 0>]}}}
     CurveIndex=$Emitter.NormalizedLoopAge
     Scale Curve=64
    )
    Use Spawn Probability=false
   )
  ]
  particleSpawnScript [
   &InitializeParticle(
    Color Mode=::NewEnumerator0
    Lifetime Max=0.4
    Lifetime Min=0.1
    Lifetime Mode=::NewEnumerator1
    Mesh Scale Mode=::NewEnumerator1
    Mesh Uniform Scale Max=1.6
    Mesh Uniform Scale Min=1.2
    UsePositionOffset=false
    Write Material Random=true
   )
   &SkeletalMeshLocation(Bone Sampling Mode=::NewEnumerator4, MaskSpawnGroup=false, Sampled Position Offset=<-50 0 0>, Skeletal Mesh=$System.Skeletal Mesh)
   &ShapeLocation(MaskSpawnGroup=false, Override Seed=false, Sphere Radius=8, UseOwnerScale=true)
   &InitialMeshOrientation(Use Rotation Vector=false)
   &AddVelocity(
    Cone Angle=25
    Cone Axis=<0 0 -1>
    Module.UseConeFalloff=false
    Module.UseVelDistribution=false
    Override Seed=false
    Velocity Mode=::NewEnumerator2
    Velocity Speed=&RandomRangeFloat(Minimum=125, Maximum=180)
   )
  ]
  particleUpdateScript [
   &ParticleState()
   &MeshRotationForce(Rotation=&VectorFromFloat(Value=50))
   &CurlNoiseForce(
    Bias Noise=false
    Noise Frequency=25
    Noise Strength=&RandomRangeFloat(Minimum=180, Maximum=250)
    Pan Noise=false
   )
   &Drag(Drag=&RandomRangeFloat(Minimum=0.8, Maximum=1.2), Use Linear Drag=true, Use Rotational Drag=true)
   &SolveForcesAndVelocity(Clamp Velocity=false, Limit Acceleration=false)
   &ScaleMeshSize(Scale Factor=&VectorFromFloat(Value=&FloatFromCurve(FloatCurve=DataInterface() {propertyValues {Curve {keys [<at 0 1 RCIM_Cubic>, {arriveTangent=-2.1534, interpMode=RCIM_Cubic, leaveTangent=-2.1534, tangentMode=RCTM_User, time=1, value=0}]}}})))
  ]
  rendererClasses [&NiagaraMeshRendererProperties]
  renderers [
   {rendererClass=&NiagaraMeshRendererProperties, rendererIndex=0}
   {rendererClass=&NiagaraMeshRendererProperties, rendererIndex=1}
  ]
 }
 MacSplatUp {
  emitterSpawnScript [
   &NM_ToonShaderParameters(RimLight_Color:LinearColor=<rgba 2 1.3024 0.2399 1>)
   &NM_OutlineParameters(OutlineColor:LinearColor=<rgba 0.0469 0.0289 0 1>, OutlineWidth=0.5)
  ]
  emitterUpdateScript [
   &EmitterState(Life Cycle Mode=::NewEnumerator1, Loop Behavior=::NewEnumerator1, Loop Delay=0.24, Loop Duration=1, UseLoopDelay=true)
   &SpawnBurst_Instantaneous(
    Spawn Count=&RandomRangeInt(Minimum=12, Maximum=8)
    Use Spawn Probability=false
    UseLoopCountLimit=false
   )
  ]
  particleSpawnScript [
   &InitializeParticle(
    Color Mode=::NewEnumerator0
    Lifetime Max=0.4
    Lifetime Min=0.1
    Lifetime Mode=::NewEnumerator1
    Mesh Scale Mode=::NewEnumerator1
    Mesh Uniform Scale Max=1.6
    Mesh Uniform Scale Min=1.2
    Position Offset=<0 0 -15>
    UsePositionOffset=false
    Write Material Random=true
   )
   &SkeletalMeshLocation(Bone Sampling Mode=::NewEnumerator4, MaskSpawnGroup=false, Sampled Position Offset=<0 10 -3>, Skeletal Mesh=$System.Skeletal Mesh)
   &ShapeLocation(
    Disc Coverage=1
    MaskSpawnGroup=false
    Offset Mode=::NewEnumerator0
    Override Seed=false
    Ring Radius=8
    Shape Primitive=::NewEnumerator4
    UseOwnerScale=true
   )
   &InitialMeshOrientation(Use Rotation Vector=false)
   &AddVelocity(
    Cone Angle=60
    Cone Axis=<0 0 1>
    Module.UseConeFalloff=false
    Module.UseVelDistribution=false
    Override Seed=false
    Rotation Angle=90
    Rotation Axis:Vector3f=<0 0 1>
    Rotation Mode=::NewEnumerator1
    Velocity Mode=::NewEnumerator2
    Velocity Speed=&RandomRangeFloat(Minimum=125, Maximum=250)
   )
  ]
  particleUpdateScript [
   &ParticleState()
   &MeshRotationForce(Rotation=&VectorFromFloat(Value=50))
   &GravityForce(Coordinate Space=::Local, Gravity=<0 0 -750>)
   &Drag(Drag=&RandomRangeFloat(Minimum=0.8, Maximum=1.2), Use Linear Drag=true, Use Rotational Drag=true)
   &SolveForcesAndVelocity(Clamp Velocity=false, Limit Acceleration=false)
   &ScaleMeshSize(Scale Factor=&VectorFromFloat(Value=&FloatFromCurve(FloatCurve=DataInterface() {propertyValues {Curve {keys [<at 0 1 RCIM_Cubic>, {arriveTangent=-2.1534, interpMode=RCIM_Cubic, leaveTangent=-2.1534, tangentMode=RCTM_User, time=1, value=0}]}}})))
  ]
  rendererClasses [&NiagaraMeshRendererProperties]
  renderers [
   {rendererClass=&NiagaraMeshRendererProperties, rendererIndex=0}
   {rendererClass=&NiagaraMeshRendererProperties, rendererIndex=1}
  ]
 }
}
summary {
 systemName=NS_LintMermaid_MacCheese
 emitters [
  {emitterName=CheeseSplat, rendererClasses [&NiagaraMeshRendererProperties]}
  {emitterName=Burst_Splashes, rendererClasses [&NiagaraSpriteRendererProperties]}
  {emitterName=MacFalling, rendererClasses [&NiagaraMeshRendererProperties]}
  {emitterName=Goop, rendererClasses [&NiagaraRibbonRendererProperties]}
  {emitterName=MacSplatUp, rendererClasses [&NiagaraMeshRendererProperties]}
  {emitterName=Goop001, rendererClasses [&NiagaraRibbonRendererProperties]}
 ]
 userVariables [
  User.bIsFrontEnd {type=NiagaraBool}
  User.bIsFrontEndPreview {type=NiagaraBool}
 ]
}
