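# Niagara system for the LintMermaid emote's sauce effects, using a gadget skeletal mesh as preview.
# SauceMain pours a short-lived red ribbon from the bottle-cap bone under gravity, coloured over loop age.
# Spiral bursts an upward-facing sprite that follows a Mac bone, animating a material parameter, with time-of-day tint.
# TrailMain and TrailBlack layer fading, tapering ribbons; three user floats set sauce, spiral and trail booleans at emitter spawn.
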NiagaraSystem /BRCosmetics/Animation/Game/MainPlayer/Emotes/LintMermaid/FX/NS_LintMermaid_SauceMain
dependencies {
 usedDataInterfaces [&NiagaraDataInterfaceCurve, &NiagaraDataInterfaceColorCurve, &NiagaraDataInterfaceSkeletalMesh]
 usedDynamicInputs []
 usedModules [&SystemState]
}
emitters {
 SauceMain {
  emitterSpawnScript [/SetVariables(Emitter.Sauce:NiagaraBool=&SetBoolByFloatComparison(A=$User.bSauce, B=0, Comparison Type=::NewEnumerator5)) {isSetParameters=true}]
  emitterUpdateScript [
   &EmitterState()
   &SpawnRate(SpawnRate=60, Use Spawn Probability=false)
  ]
  particleSpawnScript [
   &InitializeParticle(
    Color=&ColorFromCurve(
     ColorCurve=DataInterface() {propertyValues {
       AlphaCurve {keys [<at 0 0 RCIM_Cubic>,<at 0.157 0.8>,<at 0.8458 0.8>,<at 1 0 RCIM_Cubic>]}
       BlueCurve {keys [<at 0 0.0016 RCIM_Cubic>,<at 1 0.0004 RCIM_Cubic>]}
       GreenCurve {keys [<at 0 0 RCIM_Cubic>,<at 1 0 RCIM_Cubic>]}
       RedCurve {keys [<at 0 0.35 RCIM_Cubic>,<at 1 0.35 RCIM_Cubic>]}
      }}
     CurveIndex=$Emitter.NormalizedLoopAge
    )
    Lifetime=0.2
    Position Offset=<-1.2 0 -2>
    Ribbon Width=4
    Ribbon Width Mode=::NewEnumerator1
    UsePositionOffset=true
    Write Material Random=true
   )
   &SkeletalMeshLocation(
    Bone Sampling Mode=::NewEnumerator4
    MaskSpawnGroup=false
    Skeletal Mesh=DataInterface() {propertyValues {
      DefaultMesh=null
      ExcludeBoneName=root
      OptionalComponentClass=null
      PreviewMesh=/BRCosmetics/Animation/Props/Gadgets_LintMermaid/Meshes/LintMermaid_Gadget
      SkinningMode=SkinOnTheFly
      SoftSourceActor=null
      SourceMode=Default
      UvSetIndex=0
      WholeMeshLOD=-1
      bExcludeBone=false
      bReadDeformedGeometry=true
      bRequireCurrentFrameData=true
      ComponentTags []
      FilteredBones [JNT_Bottle_Cap]
      FilteredSockets []
      MeshUserParameter {parameter {name=None, typeDefHandle {registeredTypeIndex=1}, varData []}}
      SamplingRegions []
     }}
   )
  ]
  particleUpdateScript [
   &GravityForce(Coordinate Space=::Local, Gravity=<0 0 980>)
   &SolveForcesAndVelocity(Clamp Velocity=false, Limit Acceleration=false)
   &ParticleState()
   &ScaleColor(
    Scale Alpha=&FloatFromCurve(FloatCurve=DataInterface() {propertyValues {Curve {keys [
         {interpMode=RCIM_Cubic, tangentMode=RCTM_User, time=0, value=1}
         {interpMode=RCIM_Cubic, tangentMode=RCTM_User, time=0.7, value=1}
         {tangentMode=RCTM_User, time=1, value=0}
        ]}}})
    ScaleA=true
    ScaleRGB=true
   )
   &ScaleRibbonWidth(Ribbon Width Scale=&FloatFromCurve(FloatCurve=DataInterface() {propertyValues {Curve {keys [<at 0 0.3 RCIM_Cubic>,<at 1 1 RCIM_Cubic>]}}}))
  ]
  rendererClasses [&NiagaraRibbonRendererProperties]
  renderers [{rendererClass=&NiagaraRibbonRendererProperties}]
 }
 Spiral {
  emitterSpawnScript [/SetVariables(Emitter.Sauce:NiagaraBool=&SetBoolByFloatComparison(A=$User.bSauceSpiral, B=0, Comparison Type=::NewEnumerator5)) {isSetParameters=true}]
  emitterUpdateScript [
   &EmitterState()
   &SpawnBurst_Instantaneous(Use Spawn Probability=false, UseLoopCountLimit=false)
  ]
  particleSpawnScript [
   &InitializeParticle(
    Color=&FNTimeOfDayColor(Night Color=<rgba 0.3646 0 0 1>, Morning Color=<rgba 0.6406 0 0 1>, Day Color=<rgba 0.6615 0 0 1>, Evening Color=<rgba 0.3646 0 0 1>)
    Sprite Rotation Angle=25
    Sprite Rotation Mode=::NewEnumerator0
    Sprite Size Mode=::NewEnumerator3
    Uniform Sprite Size=25
    UsePositionOffset=false
    Write Material Random=true
   )
   &SpriteFacingAndAlignment(Sprite Facing=<0 0 1>, WriteAlignment=false, WriteFacing=true)
  ]
  particleUpdateScript [
   &ParticleState()
   &SkeletalMeshLocation(
    Bone Sampling Mode=::NewEnumerator4
    MaskSpawnGroup=false
    Sampled Position Offset=<0 0 1.4>
    Skeletal Mesh=DataInterface() {propertyValues {
      DefaultMesh=null
      ExcludeBoneName=root
      OptionalComponentClass=null
      PreviewMesh=&LintMermaid_Gadget
      SkinningMode=SkinOnTheFly
      SoftSourceActor=null
      SourceMode=Default
      UvSetIndex=0
      WholeMeshLOD=-1
      bExcludeBone=false
      bReadDeformedGeometry=true
      bRequireCurrentFrameData=true
      ComponentTags []
      FilteredBones [JNT_Mac_2]
      FilteredSockets []
      MeshUserParameter {parameter {name=None, typeDefHandle {registeredTypeIndex=1}, varData []}}
      SamplingRegions []
     }}
   )
   &DynamicMaterialParameters(
    Index 0 Param 1=&FloatFromCurve(FloatCurve=DataInterface() {propertyValues {Curve {keys [<at 0 1>,<at 0.8 0.125>]}}})
    Index 0 Param 2=0.15
    Module.Index0Float0Write=true
    Module.Index0Float1Write=true
    Module.Index0Float2Write=true
    Module.Index0Float3Write=true
   )
  ]
  rendererClasses [&NiagaraSpriteRendererProperties]
  renderers [{rendererClass=&NiagaraSpriteRendererProperties}]
 }
 TrailBlack {
  emitterSpawnScript [/SetVariables(Emitter.SauceTrail:NiagaraBool=&SetBoolByFloatComparison(A=$User.bSauceTrail, B=0, Comparison Type=::NewEnumerator5)) {isSetParameters=true}]
  emitterUpdateScript [
   &EmitterState()
   &SpawnRate(SpawnRate=100, Use Spawn Probability=false)
  ]
  particleSpawnScript [&InitializeParticle(
    Color=<rgba 0.0561 0.0529 0.0529 0.4>
    Lifetime=0.12
    Ribbon Facing Mode=::NewEnumerator1
    Ribbon Facing Vector:Vector3f=$Engine.Owner.SystemXAxis
    Ribbon Width=10
    Ribbon Width Mode=::NewEnumerator1
    UsePositionOffset=false
    Write Material Random=true
   )]
  particleUpdateScript [
   &ParticleState()
   &ScaleColor(
    Scale Alpha=&FloatFromCurve(FloatCurve=DataInterface() {propertyValues {Curve {keys [<at 0 1>,<at 1 0>]}}})
    ScaleA=true
    ScaleRGB=true
   )
   &ScaleRibbonWidth(Ribbon Width Scale=&FloatFromCurve(FloatCurve=DataInterface() {propertyValues {Curve {keys [<at 0 1 RCIM_Cubic>, {arriveTangent=-2.8783, interpMode=RCIM_Cubic, leaveTangent=-2.8783, tangentMode=RCTM_User, time=1, value=0}]}}}))
  ]
  rendererClasses [&NiagaraRibbonRendererProperties]
  renderers [{rendererClass=&NiagaraRibbonRendererProperties}]
 }
 TrailMain {
  emitterSpawnScript [/SetVariables(Emitter.SauceTrail:NiagaraBool=&SetBoolByFloatComparison(A=$User.bSauceTrail, B=0, Comparison Type=::NewEnumerator5)) {isSetParameters=true}]
  emitterUpdateScript [
   &EmitterState()
   &SpawnRate(SpawnRate=150, Use Spawn Probability=false)
  ]
  particleSpawnScript [&InitializeParticle(
    Color=<rgba 0.7917 0.5347 0.4412 0.1>
    Lifetime=0.12
    Ribbon Facing Mode=::NewEnumerator1
    Ribbon Facing Vector:Vector3f=$Engine.Owner.SystemXAxis
    Ribbon Width=25
    Ribbon Width Mode=::NewEnumerator1
    UsePositionOffset=false
    Write Material Random=true
   )]
  particleUpdateScript [
   &ParticleState()
   &ScaleColor(
    Scale Alpha=&FloatFromCurve(FloatCurve=DataInterface() {propertyValues {Curve {keys [<at 0 1>,<at 1 0>]}}})
    ScaleA=true
    ScaleRGB=true
   )
   &ScaleRibbonWidth(Ribbon Width Scale=&FloatFromCurve(FloatCurve=DataInterface() {propertyValues {Curve {keys [<at 0 1 RCIM_Cubic>, {arriveTangent=-2.8783, interpMode=RCIM_Cubic, leaveTangent=-2.8783, tangentMode=RCTM_User, time=1, value=0}]}}}))
  ]
  rendererClasses [&NiagaraRibbonRendererProperties]
  renderers [{rendererClass=&NiagaraRibbonRendererProperties}]
 }
}
summary {
 systemName=NS_LintMermaid_SauceMain
 emitters [
  {emitterName=TrailMain, rendererClasses [&NiagaraRibbonRendererProperties]}
  {emitterName=TrailBlack, rendererClasses [&NiagaraRibbonRendererProperties]}
  {emitterName=SauceMain, rendererClasses [&NiagaraRibbonRendererProperties]}
  {emitterName=Spiral, rendererClasses [&NiagaraSpriteRendererProperties]}
 ]
 userVariables [
  User.bSauce {type=NiagaraFloat, defaultValue=0, description=""}
  User.bSauceSpiral {type=NiagaraFloat, defaultValue=0, description=""}
  User.bSauceTrail {type=NiagaraFloat, defaultValue=0, description=""}
 ]
}
